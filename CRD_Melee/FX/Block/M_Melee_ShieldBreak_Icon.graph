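# Unlit translucent Niagara sprite material for the melee shield break icon effect, drawn from the T_Melee_ShieldBreak_M sub-UV texture.
# The texture's green channel multiplies particle color RGB and feeds the MF_FX_EmissiveAdjustments function as emissive.
# The red channel multiplies particle alpha to form a separate mask, so one packed texture supplies both brightness and transparency.
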Material /CRD_Melee/FX/Block/M_Melee_ShieldBreak_Icon
graph {
 properties {bUsedWithNiagaraSprites=true, blendMode=BLEND_Translucent, pixelDepthOffsetMode=PDOM_Legacy, shadingModel=MSM_Unlit, usedShadingModels=Unlit}
 nodes [
  n1 = ParticleColor()
  n2 = ParticleSubUV!(UVs, Tex, Apply View MipBias) {texture=&/T_Melee_ShieldBreak_M}
  MaterialFunctionCall(Emissive=Multiply(A=@n1.RGB, B=@n2.G)) {materialFunction=&MF_FX_EmissiveAdjustments}
  Multiply(A=@n1.A, B=@n2.R)
 ]
}
